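# Additive unlit mesh-particle material for a BluGlo cylinder effect, with upward or downward scrolling energy tinted by particle color.
# Inverted Fresnel, a dynamic Add parameter and panning sparkles combine with two plasma noises, with vertical tiling scaled by object bounds.
# A rounded-versus-smooth blend gives a cel look, masked at both V ends, then drives the inner-to-outer color lerp.
# The depth-fade branch does not feed that color.
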Material /Game/Missions/Secondary/BluGlo/M_BluGlo_Cylinder_02
graph {
 properties {
  bEnableResponsiveAA=true
  bUsedWithMeshParticles=true
  blendMode=BLEND_Additive
  pixelDepthOffsetMode=PDOM_Legacy
  shadingModel=MSM_Unlit
  usedShadingModels=Unlit
 }
 nodes [
  n1 = ScalarParameter() {defaultValue=1, parameterName=Fresnel Brightness}
  n2 = ScalarParameter() {defaultValue=1.5, parameterName=Fresnel Exponent}
  n3 = AppendVector(
   A=Constant() {r=1}
   B=DotProduct(A=ObjectBounds(), B=VectorParameter() {parameterName=Height DOT Tiling, defaultValue=<rgba 0 0 0.01 1>}.RGB)
  )
  n4 = Panner(Coordinate=Multiply(A=TextureCoordinate() {uTiling=10, vTiling=0.24}, B=@n3)) {bFractionalPart=true, speedY=2}
  n5 = Add(
   A=Add(
    A=DynamicParameter() {defaultValue=<rgba 0 1 1 1>, paramNames [Add,Param2,Param3,Param4]}.Add
    B=MaterialFunctionCall(Brightness=@n1, Exponent=@n2) {materialFunction=&MF_Fresnel_FX}.Fresnel Inverted
   )
   B=Multiply(
    A=TextureSample(UVs=@n4) {samplerType=SAMPLERTYPE_LinearColor, texture=/Game/Gadgets/Assets/Slow_Field/Textures/T_Slowfield_SparkleNoise_01}.R
    B=ScalarParameter() {defaultValue=0.25, parameterName=Dots Brightness}
   ) {constB=0.1}
  )
  n6 = Multiply(A=TextureCoordinate() {vTiling=0.2678}, B=@n3)
  n7 = Multiply(A=TextureCoordinate() {vTiling=0.2832}, B=@n3)
  n8 = Add(
   A=TextureSample(UVs=Panner(Coordinate=@n6) {bFractionalPart=true, speedX=0.121, speedY=2.0967}) {texture=/Game/Effects/Fort_Effects/Textures/Plasma/T_PlasmaNoise_A_01}.B
   B=TextureSample(UVs=Panner(Coordinate=@n7) {bFractionalPart=true, speedX=-0.087, speedY=1.74}) {texture=&T_PlasmaNoise_A_01}.R
  )
  n9 = Multiply(A=@n5, B=Add(A=@n5, B=@n8))
  n10 = ComponentMask(None=TextureCoordinate()) {g=true}
  n11 = Multiply(A=@n10, B=OneMinus(None=@n10))
  n12 = ScalarParameter() {defaultValue=10}
  n13 = Multiply(
   A=LinearInterpolate(
    A=MaterialFunctionCall!(In=@n9) {materialFunction=/Engine/Functions/Engine_MaterialFunctions02/Utility/Round}.Result
    B=@n5
    Alpha=ScalarParameter() {defaultValue=0.25, parameterName=Cel LERP}
   )
   B=Clamp(None=Multiply(A=@n11, B=@n12))
  )
  n14 = OneMinus(None=DepthFade(FadeDistance=ScalarParameter() {defaultValue=15, parameterName=Depth Fade Distance}))
  Add(A=@n13, B=Multiply(A=@n14, B=Add(A=@n14, B=@n8)))
  n15 = VectorParameter() {parameterName=Color Inner, defaultValue=<rgba 0.0004 0 0.1 0>}
  n16 = VectorParameter() {parameterName=Color Outer, defaultValue=<rgba 0.2691 1.8227 10 0>}
  MaterialFunctionCall(Emissive=Multiply(A=LinearInterpolate(A=@n15.RGB, B=@n16.RGB, Alpha=@n13), B=ParticleColor().RGB)) {materialFunction=&MF_FX_EmissiveAdjustments}
  ParticleColor()
 ]
}
