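# M_Invisible is an unlit, masked material in the BluGlo secondary mission folder, named for rendering geometry invisible.
# Its graph holds only one Constant node at its default value, with no textures, parameters or visible wiring to an output.
# Usage flags cover skeletal, instanced static, particle, Nanite and mesh-deformer meshes, so it fits many mesh types.
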Material /Game/Missions/Secondary/BluGlo/M_Invisible
graph {
 properties {
  bUsedWithInstancedStaticMeshes=true
  bUsedWithMeshDeformer=true
  bUsedWithMeshParticles=true
  bUsedWithNanite=true
  bUsedWithSkeletalMesh=true
  blendMode=BLEND_Masked
  pixelDepthOffsetMode=PDOM_Legacy
  shadingModel=MSM_Unlit
  usedShadingModels=Unlit
 }
 nodes [Constant()]
}
